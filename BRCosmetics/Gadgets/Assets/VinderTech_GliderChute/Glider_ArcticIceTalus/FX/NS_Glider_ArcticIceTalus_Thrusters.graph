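# Niagara effect for the Arctic Ice Talus glider: blue thruster flames and glows, warm lights, wing trails and speed lines.
# Partitioned particles follow skeletal mesh sockets, layering mesh flames with sprite glows and flickering fake lighting tinted by time of day.
# System timeline drives thruster size, fake-light intensity and gameplay opacity of the socket-pair ribbon trails.
# ForwardDot fades speed lines in gameplay, while front-end flags switch preview-specific scales, directions and opacity.
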NiagaraSystem /BRCosmetics/Gadgets/Assets/VinderTech_GliderChute/Glider_ArcticIceTalus/FX/NS_Glider_ArcticIceTalus_Thrusters
dependencies {
 usedDataInterfaces [
  &NiagaraDataInterfaceCurve
  &NiagaraDataInterfaceColorCurve
  &NiagaraDataInterfaceVector2DCurve
  &NiagaraDataInterfaceVectorCurve
 ]
 usedDynamicInputs [&Boolean_AndOperation, &SetBoolByFloatComparison]
 usedModules [
  &SystemState
  &Timeline
  &self:SystemSpawnScript.SystemScriptSource.SystemScriptGraph.NiagaraNodeAssignment_1.SetVariables_#1
 ]
}
emitters {
 AirThrusters {
  emitterUpdateScript [
   &EmitterState()
   &SpawnRate(SpawnRate=32, Use Spawn Probability=false)
  ]
  particleSpawnScript [
   &PartitionParticles(Number Of Partitions=4)
   &InitializeParticle(
    Color=<rgba 1 1 1 0.1>
    Mesh Scale=&MakeCustomVectorFromBool(
     Bool=&SetBoolByIntComparison(A=$Particles.PartitionParticles.Partition, B=1, Comparison Type=::NewEnumerator0)
     True Vector=<0.07 0.07 0.15>
     False Vector=<0.12 0.12 0.25>
    )
    Mesh Scale Mode=::NewEnumerator4
    UsePositionOffset=false
    Write Material Random=true
   )
   &InitialMeshOrientation(
    Mesh Axis To Orient=<0 1 0>
    Mesh Orientation Mode=::NewEnumerator2
    Orientation Vector=<0 0 -1>
    Rotation=&RandomRangeVector(Minimum=<0 0 0>, Maximum=<0 0 1>)
    Use Rotation Vector=true
   )
  ]
  particleUpdateScript [
   &ParticleState()
   &SkeletalMeshLocation(
    Bone / Socket Index=$Particles.PartitionParticles.Partition
    MaskSpawnGroup=false
    Mesh Sampling Type=::NewEnumerator5
    Sampled Position Offset=&MakeCustomVectorFromBool(
     Bool=&SetBoolByIntComparison(A=$Particles.PartitionParticles.Partition, B=1, Comparison Type=::NewEnumerator0)
     True Vector=<0 -15 0>
     False Vector=<0 -30 0>
    )
    Skeletal Mesh=$User.SkeletalMesh
    Socket Sampling Mode=::NewEnumerator5
   )
   &ScaleMeshSize(Scale Factor=&MakeVector(
     X=1
     Y=1
     Z=&Lerp_Float(
      A=0.6
      B=1
      Alpha=&ClampFloat(Float=&Add_Float(A=$System.Timeline.NormalizedTime, B=$System.Size scale))
     )
    ))
   &ScaleColor(
    Scale Alpha=&Lerp_Float(A=0, B=0.2, Alpha=$Particles.NormalizedAge)
    Scale RGB=&MakeVectorFromLinearColorRGB(Color=&ColorFromCurve(ColorCurve=DataInterface() {propertyValues {
        AlphaCurve {keys [<at 0 1 RCIM_Cubic>,<at 1 1 RCIM_Cubic>]}
        BlueCurve {keys [<at 0.1329 1 RCIM_Cubic>,<at 0.5656 0.1 RCIM_Cubic>]}
        GreenCurve {keys [<at 0.1329 0.2605 RCIM_Cubic>,<at 0.5656 0.0167 RCIM_Cubic>]}
        RedCurve {keys [<at 0.1329 0 RCIM_Cubic>,<at 0.5656 0 RCIM_Cubic>]}
       }}))
    ScaleA=true
    ScaleRGB=false
   )
   /SetVariables(Particles.ColorAlt:LinearColor=&MakeLinearColorFromVectorAndFloat(
     "Vector (RGB)"=&MakeVectorFromLinearColorRGB(Color=&ColorFromCurve(ColorCurve=DataInterface() {propertyValues {
         AlphaCurve {keys [<at 0 1 RCIM_Cubic>,<at 1 1 RCIM_Cubic>]}
         BlueCurve {keys [
           <at 0.083 1 RCIM_Cubic>
           {arriveTangent=-2.419, interpMode=RCIM_Cubic, time=0.2491, value=0.1}
           {interpMode=RCIM_Cubic, leaveTangent=1.4033, time=0.5429, value=0.0469}
          ]}
         GreenCurve {keys [
           <at 0.083 0 RCIM_Cubic>
           {arriveTangent=0.0614, interpMode=RCIM_Cubic, time=0.2491, value=0.0565}
           {interpMode=RCIM_Cubic, leaveTangent=0.7928, time=0.5429, value=0.0242}
          ]}
         RedCurve {keys [<at 0.083 0.5929 RCIM_Cubic>, {arriveTangent=-1.5047, interpMode=RCIM_Cubic, time=0.2491, value=0}, <at 0.5429 0 RCIM_Cubic>]}
        }}))
     "Float (Alpha)"=&Lerp_Float(A=0, B=0.05, Alpha=$Particles.NormalizedAge)
    )) {isSetParameters=true}
  ]
  rendererClasses [&NiagaraMeshRendererProperties]
  renderers [
   {rendererClass=&NiagaraMeshRendererProperties, rendererIndex=0}
   {rendererClass=&NiagaraMeshRendererProperties, rendererIndex=1}
  ]
 }
 FakeLighting {
  emitterUpdateScript [
   &EmitterState()
   &SpawnBurst_Instantaneous(Spawn Count=4, Use Spawn Probability=false, UseLoopCountLimit=true)
  ]
  particleSpawnScript [
   &PartitionParticles(Number Of Partitions=4)
   &InitializeParticle(Color=<rgba 1 1 1 0.5>, Sprite Size Mode=::NewEnumerator3, Uniform Sprite Size=55, UsePositionOffset=false, Write Material Random=true)
   &DynamicMaterialParameters(
    Index 0 Param 1=1
    Index 0 Param 2=1
    Index 0 Param 3=1
    Index 0 Param 4=1
    Module.Param0WriteEnabled=true
    Module.Param1WriteEnabled=true
    Module.Param2WriteEnabled=true
    Module.Param3WriteEnabled=true
   )
  ]
  particleUpdateScript [
   &ParticleState(
    Kill Particles When Lifetime Has Elapsed=false
    Let Infinitely Lived Particles Die When Emitter Deactivates=true
    Loop Particles Lifetime=true
   )
   &SkeletalMeshLocation(
    Bone / Socket Index=$Particles.PartitionParticles.Partition
    MaskSpawnGroup=false
    Mesh Sampling Type=::NewEnumerator5
    Sampled Position Offset=&Lerp_Vector(A=<1 10 10>, B=<-1 10 10>, Alpha=&MakeFloatFromInt(Int=$Particles.PartitionParticles.Partition))
    Skeletal Mesh=$User.SkeletalMesh
    Socket Sampling Mode=::NewEnumerator5
   )
   &CameraOffset()
   &ScaleSpriteSize(
    Scale Sprite Size Mode=::NewEnumerator0
    Uniform Scale Factor=&MakeCustomFloatFromBool(
     Bool=&SetBoolByIntComparison(A=$Particles.PartitionParticles.Partition, B=1, Comparison Type=::NewEnumerator0)
     True Float=0
     False Float=1
    )
   )
   &ScaleColor(
    Scale Alpha=&Lerp_Float(
     A=0.25
     B=&Multiply_Float(
      A=1
      B=&FloatFromCurve(
       FloatCurve=DataInterface() {propertyValues {Curve {keys [
           <at 0 1 RCIM_Cubic>
           {arriveTangent=0.1177, interpMode=RCIM_Cubic, leaveTangent=0.1177, time=0.0746, value=0.8987}
           {arriveTangent=-0.0227, interpMode=RCIM_Cubic, leaveTangent=-0.0227, tangentMode=RCTM_User, time=0.1682, value=1.0198}
           {arriveTangent=-0.7897, interpMode=RCIM_Cubic, leaveTangent=-0.7897, time=0.2019, value=0.7756}
           {arriveTangent=0.8345, interpMode=RCIM_Cubic, leaveTangent=0.8345, time=0.2748, value=0.9356}
           {arriveTangent=0.5127, interpMode=RCIM_Cubic, leaveTangent=0.5127, time=0.3464, value=0.8961}
           {arriveTangent=-0.6889, interpMode=RCIM_Cubic, leaveTangent=-0.6889, time=0.4074, value=1.0036}
           {arriveTangent=-0.8737, interpMode=RCIM_Cubic, leaveTangent=-0.8737, time=0.4365, value=0.834}
           {arriveTangent=-0.4832, interpMode=RCIM_Cubic, leaveTangent=-0.4832, time=0.4871, value=0.9339}
           {arriveTangent=0.1715, interpMode=RCIM_Cubic, leaveTangent=0.1715, time=0.5331, value=0.7873}
           {arriveTangent=0.795, interpMode=RCIM_Cubic, leaveTangent=0.795, time=0.6112, value=0.9552}
           {arriveTangent=0.9015, interpMode=RCIM_Cubic, leaveTangent=0.9015, time=0.69, value=0.912}
           {arriveTangent=-0.4254, interpMode=RCIM_Cubic, leaveTangent=-0.4254, time=0.7536, value=1.0835}
           {arriveTangent=-1.4701, interpMode=RCIM_Cubic, leaveTangent=-1.4701, time=0.7995, value=0.8654}
           {arriveTangent=-0.4912, interpMode=RCIM_Cubic, leaveTangent=-0.4912, time=0.8674, value=0.9161}
           {arriveTangent=0.6326, interpMode=RCIM_Cubic, leaveTangent=0.6326, time=0.9138, value=0.8093}
           {interpMode=RCIM_Cubic, leaveTangent=0.1355, time=1, value=1}
          ]}}}
       Scale Curve=&FNTimeOfDayScalar(Night Scale=0.4, Morning Scale=1, Evening Scale=0.5)
      )
     )
     Alpha=&ClampFloat(Float=&Add_Float(
       A=$System.Timeline.NormalizedTime
       B=&FloatFromCurve(
        FloatCurve=DataInterface() {propertyValues {Curve {keys [
            {tangentMode=RCTM_User, time=0, value=1}
            {arriveTangent=-0.0405, interpMode=RCIM_Cubic, leaveTangent=-0.0405, tangentMode=RCTM_User, time=1.2, value=1}
            {arriveTangent=-8.0089, interpMode=RCIM_Cubic, leaveTangent=-8.0089, tangentMode=RCTM_User, time=1.5, value=0}
           ]}}}
        CurveIndex=$System.Age
       )
      ))
    )
    Scale RGB=&MakeVectorFromLinearColorRGB(Color=&FNTimeOfDayColor(
      Night Color=<rgba 0.0966 0.0966 0.786 1>
      Morning Color=<rgba 0.0966 0.3264 0.786 1>
      Day Color=<rgba 0.0966 0.3264 0.786 1>
      Evening Color=<rgba 0.0966 0.0966 0.786 1>
     ))
    ScaleA=true
    ScaleRGB=true
   )
   /SetVariables(
    Particles.CoronaColor:LinearColor=&MultiplyLinearColor(
     Linear Color=&FNTimeOfDayColor(
      Night Color=<rgba 0.0583 0 0.7 1>
      Morning Color=<rgba 0.1843 0.1843 1.5 1>
      Day Color=<rgba 0.2458 0.2458 2 1>
      Evening Color=<rgba 0.0583 0 0.7 1>
     )
     Scale Factor=&MakeFloat4FromVectorAndFloat(
      "Vector (RGB)"=<2 2 2>
      "Float (Alpha)"=&Lerp_Float(
       A=0.4
       B=1
       Alpha=&ClampFloat(Float=&Add_Float(
         A=$System.Timeline.NormalizedTime
         B=&FloatFromCurve(
          FloatCurve=DataInterface() {propertyValues {Curve {keys [
              {tangentMode=RCTM_User, time=0, value=1}
              {arriveTangent=-0.0405, interpMode=RCIM_Cubic, leaveTangent=-0.0405, tangentMode=RCTM_User, time=1.2, value=1}
              {arriveTangent=-8.0089, interpMode=RCIM_Cubic, leaveTangent=-8.0089, tangentMode=RCTM_User, time=1.5, value=0}
             ]}}}
          CurveIndex=$System.Age
         )
        ))
      )
     )
    )
    Particles.CoronaPosition:NiagaraPosition=&Add_Vector(
     A=Linked() {linkedVariable {name=Output.SkeletalMeshLocation.SampledPositionWithoutOffset, type=NiagaraPosition}}
     B=&RotateVectorByQuaternion(Vector To Rotate=<0 7 0>, Quaternion=$Particles.SkeletalMeshLocation.SampledRotation)
    )
    Particles.CoronaSize:Vector2f=<60 60>
   ) {isSetParameters=true}
  ]
  rendererClasses [&NiagaraSpriteRendererProperties]
  renderers [
   {rendererClass=&NiagaraSpriteRendererProperties}
   {rendererClass=&NiagaraSpriteRendererProperties, rendererIndex=1}
  ]
 }
 FireMesh {
  emitterUpdateScript [
   &EmitterState()
   &SpawnRate(SpawnRate=40, Use Spawn Probability=false)
  ]
  particleSpawnScript [
   &PartitionParticles(Number Of Partitions=4)
   &InitializeParticle(
    Color=<rgba 1 1 1 0.2>
    Lifetime Max=0.27
    Lifetime Min=0.14
    Lifetime Mode=::NewEnumerator1
    Mesh Scale=&MakeCustomVectorFromBool(
     Bool=&SetBoolByIntComparison(A=$Particles.PartitionParticles.Partition, B=1, Comparison Type=::NewEnumerator0)
     True Vector=<0.05 0.05 0.5>
     False Vector=<0.08 0.08 1>
    )
    Mesh Scale Mode=::NewEnumerator4
    UsePositionOffset=false
    Write Material Random=true
   )
   &InitialMeshOrientation(
    Mesh Axis To Orient=<0 1 0>
    Mesh Orientation Mode=::NewEnumerator2
    Orientation Vector=<0 0 1>
    Rotation=&RandomRangeVector(Minimum=<0 0 -1>, Maximum=<0 0 1>)
    Use Rotation Vector=true
   )
   &CameraOffset(Camera Offset Amount=5)
  ]
  particleUpdateScript [
   &ParticleState()
   &SkeletalMeshLocation(
    Bone / Socket Index=$Particles.PartitionParticles.Partition
    MaskSpawnGroup=false
    Mesh Sampling Type=::NewEnumerator5
    Sampled Position Offset=<0 5 0>
    Skeletal Mesh=$User.SkeletalMesh
    Socket Sampling Mode=::NewEnumerator5
   )
   &ScaleMeshSize(Scale Factor=&MakeVector(
     X=&FloatFromCurve(FloatCurve=DataInterface() {propertyValues {Curve {keys [{tangentMode=RCTM_User, time=0, value=0.75}, {tangentMode=RCTM_User, time=1, value=1.2}]}}})
     Y=&FloatFromCurve(FloatCurve=DataInterface() {propertyValues {Curve {keys [{tangentMode=RCTM_User, time=0, value=0.75}, {tangentMode=RCTM_User, time=1, value=1.2}]}}})
     Z=&FloatFromCurve(
      FloatCurve=DataInterface() {propertyValues {Curve {keys [<at 0 0>,<at 1 1>]}}}
      Scale Curve=&Lerp_Float(
       A=1.5
       B=3.5
       Alpha=&ClampFloat(Float=&Add_Float(A=$System.Timeline.NormalizedTime, B=$System.Size scale))
      )
     )
    ))
   &ScaleColor(
    Scale Alpha=&FloatFromCurve(FloatCurve=DataInterface() {propertyValues {Curve {keys [
         {arriveTangent=9.0725, interpMode=RCIM_Cubic, leaveTangent=9.0725, tangentMode=RCTM_User, time=0, value=0}
         {arriveTangent=-1.6443, interpMode=RCIM_Cubic, leaveTangent=-1.6443, tangentMode=RCTM_User, time=0.1408, value=0.9847}
         {arriveTangent=-0.0405, interpMode=RCIM_Cubic, leaveTangent=-0.0405, tangentMode=RCTM_User, time=1, value=0}
        ]}}})
    Scale RGB=&MakeVectorFromLinearColorRGB(Color=&ColorFromCurve(ColorCurve=DataInterface() {propertyValues {
        AlphaCurve {keys [<at 0 1 RCIM_Cubic>]}
        BlueCurve {keys [{arriveTangent=-1.6041, interpMode=RCIM_Cubic, time=0.17, value=2}, <at 0.4736 2>, <at 1 1>]}
        GreenCurve {keys [{arriveTangent=-1.6041, interpMode=RCIM_Cubic, time=0.17, value=0.7333}, <at 0.4736 0.0876>, <at 1 0>]}
        RedCurve {keys [{arriveTangent=-1.6041, interpMode=RCIM_Cubic, time=0.17, value=0}, <at 0.4736 0.0563>, <at 1 0.0333>]}
       }}))
    ScaleA=true
    ScaleRGB=true
   )
  ]
  rendererClasses [&NiagaraMeshRendererProperties]
  renderers [{rendererClass=&NiagaraMeshRendererProperties, rendererIndex=0}]
 }
 FlameMesh {
  emitterUpdateScript [
   &EmitterState()
   &SpawnRate(SpawnRate=40, Use Spawn Probability=false)
  ]
  particleSpawnScript [
   &PartitionParticles(Number Of Partitions=4)
   &InitializeParticle(
    Color=<rgba 1 1 1 0.07>
    Mesh Scale=&MakeCustomVectorFromBool(
     Bool=&SetBoolByIntComparison(A=$Particles.PartitionParticles.Partition, B=1, Comparison Type=::NewEnumerator0)
     True Vector=<0.07 0.07 0.15>
     False Vector=<0.15 0.15 0.25>
    )
    Mesh Scale Mode=::NewEnumerator4
    UsePositionOffset=false
    Write Material Random=true
   )
   &DynamicMaterialParameters(
    Index 0 Param 1=&RandomRangeFloat(Minimum=0, Maximum=1)
    Index 0 Param 4=&Lerp_Float(A=0, B=20, Alpha=0.5)
    Module.Param0WriteEnabled=true
    Module.Param1WriteEnabled=false
    Module.Param2WriteEnabled=false
    Module.Param3WriteEnabled=true
   )
   &InitialMeshOrientation(
    Mesh Axis To Orient=<0 1 0>
    Mesh Orientation Mode=::NewEnumerator2
    Orientation Vector=<0 0 -1>
    Rotation=&RandomRangeVector(Minimum=<0 0 0>, Maximum=<0 0 1>)
    Use Rotation Vector=true
   )
  ]
  particleUpdateScript [
   &ParticleState()
   &SkeletalMeshLocation(
    Bone / Socket Index=$Particles.PartitionParticles.Partition
    MaskSpawnGroup=false
    Mesh Sampling Type=::NewEnumerator5
    Sampled Position Offset=&MakeCustomVectorFromBool(
     Bool=&SetBoolByIntComparison(A=$Particles.PartitionParticles.Partition, B=1, Comparison Type=::NewEnumerator0)
     True Vector=<0 -15 0>
     False Vector=<0 -30 0>
    )
    Skeletal Mesh=$User.SkeletalMesh
    Socket Sampling Mode=::NewEnumerator5
   )
   &ScaleColor(
    Scale RGB=&MakeVectorFromLinearColorRGB(Color=&ColorFromCurve(ColorCurve=DataInterface() {propertyValues {
        AlphaCurve {keys [<at 0 1 RCIM_Cubic>]}
        BlueCurve {keys [<at 0.0203 10>, <at 0.0654 10>, {arriveTangent=-15.6046, time=0.1787, value=1}, <at 0.6131 1>]}
        GreenCurve {keys [<at 0.0203 5.7587>, <at 0.0654 0.4985>, {arriveTangent=-5.6465, time=0.1787, value=0.1523}, <at 0.6131 0.0935>]}
        RedCurve {keys [<at 0.0203 0>, <at 0.0654 0.4985>, {arriveTangent=-0.716, time=0.1787, value=0.0403}, <at 0.6131 0.0622>]}
       }}))
    ScaleA=true
    ScaleRGB=true
   )
   &ScaleMeshSize(Scale Factor=&FN_FrontEndVector3Modifier(
     Gameplay=&VectorFromCurve(
      VectorCurve=DataInterface() {propertyValues {
        XCurve {keys [<at 0 1>,<at 1 1>]}
        YCurve {keys [<at 0 1>,<at 1 1>]}
        ZCurve {keys [<at 0.9 1 RCIM_Cubic>, {arriveTangent=-4.873, interpMode=RCIM_Cubic, leaveTangent=-4.873, tangentMode=RCTM_User, time=1, value=0}]}
       }}
      Scale Curve=&Lerp_Vector(
       A=<0.47 0.47 1.2>
       B=<0.525 0.525 1.8>
       Alpha=&ClampFloat(Float=&Add_Float(A=$System.Timeline.NormalizedTime, B=$System.Size scale))
      )
     )
     Front End=<0.4 0.4 3>
    ))
  ]
  rendererClasses [&NiagaraMeshRendererProperties]
  renderers [{rendererClass=&NiagaraMeshRendererProperties, rendererIndex=0}]
 }
 IdleGlow {
  emitterUpdateScript [
   &EmitterState()
   &SpawnRate(SpawnRate=20, Use Spawn Probability=false)
  ]
  particleSpawnScript [
   &PartitionParticles(Number Of Partitions=4)
   &InitializeParticle(
    Color=<rgba 0.0471 1.0601 3 0.5>
    Lifetime Max=0.5
    Lifetime Min=0.75
    Lifetime Mode=::NewEnumerator1
    Sprite Size Mode=::NewEnumerator3
    Uniform Sprite Size=&MakeCustomFloatFromBool(
     Bool=&SetBoolByIntComparison(A=$Particles.PartitionParticles.Partition, B=1, Comparison Type=::NewEnumerator0)
     True Float=18
     False Float=24
    )
    UsePositionOffset=false
    Write Material Random=true
   )
   &CameraOffset(Camera Offset Amount=5)
   /SetVariables(Particles.FacingCamera:Vector3f=<0 1 0>) {isSetParameters=true}
   /SetVariables(Particles.CamreaOffsetLow=4) {isSetParameters=true}
  ]
  particleUpdateScript [
   &ParticleState()
   &SkeletalMeshLocation(
    Bone / Socket Index=$Particles.PartitionParticles.Partition
    MaskSpawnGroup=false
    Mesh Sampling Type=::NewEnumerator5
    Skeletal Mesh=$User.SkeletalMesh
    Socket Sampling Mode=::NewEnumerator5
   )
   &DynamicMaterialParameters(
    Index 0 Param 1=50
    Module.Param0WriteEnabled=true
    Module.Param1WriteEnabled=true
    Module.Param2WriteEnabled=true
    Module.Param3WriteEnabled=true
   )
   &ScaleColor(
    Scale Alpha=&FloatFromCurve(FloatCurve=DataInterface() {propertyValues {Curve {keys [
         {arriveTangent=6.8336, interpMode=RCIM_Cubic, leaveTangent=6.8336, tangentMode=RCTM_User, time=0, value=0}
         <at 0.2 1 RCIM_Cubic>
         {arriveTangent=-1.25, interpMode=RCIM_Cubic, leaveTangent=-1.25, time=0.75, value=0.8}
         {arriveTangent=-4.31, interpMode=RCIM_Cubic, leaveTangent=-4.31, tangentMode=RCTM_User, time=1, value=0}
        ]}}})
    ScaleA=true
    ScaleRGB=true
   )
  ]
  rendererClasses [&NiagaraSpriteRendererProperties]
  renderers [
   {rendererClass=&NiagaraSpriteRendererProperties}
   {rendererClass=&NiagaraSpriteRendererProperties, rendererIndex=1}
  ]
 }
 LightsGlow {
  emitterUpdateScript [
   &EmitterState()
   &SpawnBurst_Instantaneous(Spawn Count=2, Use Spawn Probability=false, UseLoopCountLimit=true)
  ]
  particleSpawnScript [
   &PartitionParticles()
   &InitializeParticle(
    Color=<rgba 1.5 0.3634 0.1642 0.15>
    Sprite Size Mode=::NewEnumerator3
    Uniform Sprite Size=20
    UsePositionOffset=false
    Write Material Random=true
   )
   &CameraOffset(Camera Offset Amount=-5)
  ]
  particleUpdateScript [
   &ParticleState(Kill Particles When Lifetime Has Elapsed=false, Let Infinitely Lived Particles Die When Emitter Deactivates=true)
   &SkeletalMeshLocation(
    Bone / Socket Index=&Add_Integer(A=$Particles.PartitionParticles.Partition, B=4)
    MaskSpawnGroup=false
    Mesh Sampling Type=::NewEnumerator5
    Skeletal Mesh=$User.SkeletalMesh
    Socket Sampling Mode=::NewEnumerator5
   )
   &DynamicMaterialParameters(
    Index 0 Param 1=100
    Module.Param0WriteEnabled=true
    Module.Param1WriteEnabled=true
    Module.Param2WriteEnabled=true
    Module.Param3WriteEnabled=true
   )
  ]
  rendererClasses [&NiagaraSpriteRendererProperties]
  renderers [{rendererClass=&NiagaraSpriteRendererProperties}]
 }
 SideWings {
  emitterSpawnScript [&FN_RibbonIDOffsetUtility()]
  emitterUpdateScript [
   &EmitterState()
   &FN_SpawnAnimTrail(Number of Anim Trails=2)
  ]
  particleSpawnScript [
   &FN_InitializeAnimTrailBoneIndecies(First Bone or Socket Index=7, Second Bone or Socket Index=6)
   &FN_InitializeAnimTrailBoneIndecies(AnimTrail Selection=1, First Bone or Socket Index=9, Second Bone or Socket Index=8) {module=FN_InitializeAnimTrailBoneIndecies001}
   &FN_InitializeAnimTrail(AnimTrailWidth=1.025, Color=<rgba 1 1 1 0.25>, Lifetime=0.1)
   &AddVelocity(Velocity=&MakeCustomVectorFromBool(Bool=$User.bIsFrontEnd, True Vector=<0 -2500 0>, False Vector=<0 2500 0>))
  ]
  particleUpdateScript [
   &SolveForcesAndVelocity(Clamp Velocity=false, Limit Acceleration=false)
   &FN_UpdateAnimTrail(SkeletalMesh=$User.SkeletalMesh)
   /SetVariables(Particles.RibbonU0Override=$Particles.NormalizedAge) {isSetParameters=true}
   &ScaleColor(
    Scale Alpha=&FN_FrontEndFloatModifier(Gameplay=&Lerp_Float(A=0, B=0.8, Alpha=$System.Timeline.NormalizedTime), Front End=1)
    ScaleA=true
    ScaleRGB=true
   )
   /SetVariables(Particles.LightTrailColor:LinearColor=&MakeLinearColorFromVectorAndFloat(
     "Vector (RGB)"=&MakeVectorFromLinearColorRGB(Color=<rgba 0.5833 0.1805 0.0299 1>)
     "Float (Alpha)"=&FN_FrontEndFloatModifier(
      Gameplay=&Lerp_Float(
       A=0
       B=&FloatFromCurve(FloatCurve=DataInterface() {propertyValues {Curve {keys [{arriveTangent=-0.5261, interpMode=RCIM_Cubic, leaveTangent=-0.5261, tangentMode=RCTM_User, time=0, value=0.09}, <at 0.3 0 RCIM_Cubic>]}}})
       Alpha=$System.Timeline.NormalizedTime
      )
      Front End=0.2
     )
    )) {isSetParameters=true}
  ]
  rendererClasses [&NiagaraRibbonRendererProperties]
  renderers [
   {rendererClass=&NiagaraRibbonRendererProperties}
   {rendererClass=&NiagaraRibbonRendererProperties, rendererIndex=1}
  ]
 }
 SpeedLines {
  emitterUpdateScript [
   &EmitterState()
   &SpawnRate(
    SpawnRate=&MakeCustomFloatFromBool(Bool=$User.bIsFrontEndPreview, True Float=50, False Float=120)
    Use Spawn Probability=true
   )
  ]
  particleSpawnScript [
   &InitializeParticle(
    Lifetime Max=0.16
    Lifetime Min=0.1
    Lifetime Mode=::NewEnumerator1
    Sprite Size Max=<3 10>
    Sprite Size Min=<2 2>
    Sprite Size Mode=::NewEnumerator2
    UsePositionOffset=false
    Write Material Random=true
   )
   &TorusLocation(
    Align Torus With Axis=true
    Handle Radius=0
    Large Radius:NiagaraFloat=&RandomRangeFloat(Minimum=150, Maximum=300)
    MaskSpawnGroup=false
    Module.Override Randomness=false
    Offset=<0 -150 0>
    Torus Axis=<0 1 0>
   )
   &AddVelocityFromPoint(
    Module.Use Falloff Distance=false
    Origin Offset=&MakeCustomVectorFromBool(Bool=$User.bIsFrontEndPreview, True Vector=<0 500 0>, False Vector=<0 -500 0>)
    Velocity Strength=600
   )
   &AddVelocity(Velocity=&MakeCustomVectorFromBool(
     Bool=$User.bIsFrontEndPreview
     True Vector=&RandomRangeVector(Minimum=<0 -300 0>, Maximum=<0 -700 0>)
     False Vector=&RandomRangeVector(Minimum=<0 300 0>, Maximum=<0 700 0>)
    ))
  ]
  particleUpdateScript [
   &SolveForcesAndVelocity(Clamp Velocity=false, Limit Acceleration=false)
   &ParticleState()
   &ScaleSpriteSize(Scale Factor=&Vector2DFromCurve(
     Vector2Curve=DataInterface() {propertyValues {XCurve {keys [<at 0 0.5 RCIM_Cubic>,<at 0.5 1 RCIM_Cubic>]}, YCurve {keys [<at 0 0 RCIM_Cubic>,<at 0.5 1 RCIM_Cubic>]}}}
     ScaleCurve=&Vector2DFromFloat(Value=1)
    ))
   &ScaleSpriteSizeBySpeed(Max Scale Factor=<1 50>, Min Scale Factor=<0.01 0.05>, Velocity Threshold=1)
   &ScaleColor(
    Scale Alpha=&FloatFromCurve(
     FloatCurve=DataInterface() {propertyValues {Curve {keys [
         <at 0 0 RCIM_Cubic>
         <at 0.3759 1.0207>
         {arriveTangent=-2.221, interpMode=RCIM_Cubic, leaveTangent=-2.221, tangentMode=RCTM_User, time=1, value=0}
        ]}}}
     Scale Curve=&MakeCustomFloatFromBool(
      Bool=$User.bIsFrontEndPreview
      True Float=&MakeCustomFloatFromBool(Bool=$User.bIsFullyDeployed, True Float=1, False Float=0)
      False Float=&Lerp_Float(A=0, B=1, Alpha=$User.ForwardDot)
     )
    )
    Scale RGB=&VectorFromFloat(Value=1)
    ScaleA=true
    ScaleRGB=true
   )
  ]
  rendererClasses [&NiagaraSpriteRendererProperties]
  renderers [{rendererClass=&NiagaraSpriteRendererProperties}]
 }
}
summary {
 systemName=NS_Glider_ArcticIceTalus_Thrusters
 emitters [
  {emitterName=FireMesh, rendererClasses [&NiagaraMeshRendererProperties]}
  {emitterName=IdleGlow, rendererClasses [&NiagaraSpriteRendererProperties]}
  {emitterName=SpeedLines, rendererClasses [&NiagaraSpriteRendererProperties]}
  {emitterName=SideWings, rendererClasses [&NiagaraRibbonRendererProperties]}
  {emitterName=AirThrusters, rendererClasses [&NiagaraMeshRendererProperties]}
  {emitterName=FlameMesh, rendererClasses [&NiagaraMeshRendererProperties]}
  {emitterName=LightsGlow, rendererClasses [&NiagaraSpriteRendererProperties]}
  {emitterName=FakeLighting, rendererClasses [&NiagaraSpriteRendererProperties]}
 ]
 userVariables [
  User.bIsFrontEnd {type=NiagaraBool}
  User.bIsFrontEndPreview {type=NiagaraBool}
  User.bIsFullyDeployed {type=NiagaraBool}
  User.Delay {type=NiagaraFloat, defaultValue=0, description=""}
  User.EyeglowColor {type=LinearColor, description="", defaultValue {struct=LinearColor, value=<rgba 1 1 1 1>}}
  User.ForwardDot {type=NiagaraFloat, defaultValue=0, description=""}
  User.RibbonIDOffset {type=NiagaraInt32}
  User.RightDot {type=NiagaraFloat, defaultValue=0, description=""}
  User.SkeletalMesh {
   type=NiagaraDataInterfaceSkeletalMesh
   defaultValue {
    struct=NiagaraExt_VariableValue_DataInterface
    value {dataInterface=&self:NiagaraDataInterfaceSkeletalMesh_0}
   }
  }
 ]
}
