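# Physics asset for the Arctic Ice Talus glider cosmetic's skeletal mesh, giving its main bones simple collision shapes.
# Six capsules cover the body, handlebar, mirrored left and right wings, and a piston bone on each side.
# Four constraints join both wings and both pistons directly to the body, all sharing the same swing and twist limit angles.
# The handlebar has a capsule but no constraint linking it to the other bodies.
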PhysicsAsset /BRCosmetics/Gadgets/Assets/VinderTech_GliderChute/Glider_ArcticIceTalus/Meshes/ArcticIceTalus_Glider_Physics
bodies {
 JNT_Body [{shapeType=Capsule, length=146.3806, radius=135.4212, center=<0.0005 24.0294 33.1091>, rotation=<rot 0 179.9989 -88.9531>}]
 JNT_Handlebar [{shapeType=Capsule, length=57.6093, radius=50.2872, center=<-0.0001 -2.7706 -0.0556>, rotation=<rot 0 90.0068 -89.9985>}]
 JNT_L_Piston2_01 [{shapeType=Capsule, length=16.5877, radius=10.0887, center=<13.1348 0.5956 -2.558>, rotation=<rot 0 -90.5543 -79.3018>}]
 JNT_L_Wing [{shapeType=Capsule, length=151.7286, radius=59.0004, center=<-119.428 36.0535 5.1614>, rotation=<rot 0 -110.0259 -86.8149>}]
 JNT_R_Piston2_01 [{shapeType=Capsule, length=16.5877, radius=10.0887, center=<-13.1348 -0.5956 2.558>, rotation=<rot 0 -90.5543 -79.3018>}]
 JNT_R_Wing [{shapeType=Capsule, length=151.7286, radius=59.0004, center=<119.4281 -36.0535 -5.1614>, rotation=<rot 0 -110.0259 -86.8149>}]
}
constraints [
 {bone1Name=JNT_L_Wing, bone2Name=JNT_Body, swing1LimitDegrees=45, swing2LimitDegrees=45, twistLimitDegrees=45}
 {bone1Name=JNT_R_Wing, bone2Name=JNT_Body, swing1LimitDegrees=45, swing2LimitDegrees=45, twistLimitDegrees=45}
 {bone1Name=JNT_L_Piston2_01, bone2Name=JNT_Body, swing1LimitDegrees=45, swing2LimitDegrees=45, twistLimitDegrees=45}
 {bone1Name=JNT_R_Piston2_01, bone2Name=JNT_Body, swing1LimitDegrees=45, swing2LimitDegrees=45, twistLimitDegrees=45}
]
